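# Niagara effect for the Historian glider cosmetic: a brief one-shot whirl made of a single mesh particle.
# The Ring emitter bursts one particle that re-samples the user-supplied skeletal mesh every update, taking its position and orientation with a slight downward offset.
# Over its short life, curves grow the mesh from zero, shift its color from warm orange to white while fading alpha, and ramp a dynamic material parameter.
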NiagaraSystem /BRCosmetics/Gadgets/Assets/VinderTech_GliderChute/Glider_Historian/FX/NS_Glider_Historian_Whirl
dependencies {
 usedDataInterfaces [&NiagaraDataInterfaceColorCurve, &NiagaraDataInterfaceCurve]
 usedDynamicInputs []
 usedModules [&SystemState]
}
emitters {Ring {
  emitterUpdateScript [
   &EmitterState()
   &SpawnBurst_Instantaneous(Spawn Count=1, Spawn Time=0.02, Use Spawn Probability=false)
  ]
  particleSpawnScript [/Niagara/Modules/Spawn/Initialization/InitializeParticle(
    Lifetime=0.55
    Mesh Scale=&VectorFromFloat(Value=1.6)
    Module.Write Color=true
    Module.Write Lifetime=true
    Module.Write Mass=true
    Module.Write Material Random=true
    Module.Write Position=false
    Module.Write Scale=true
    Module.Write SpriteRotation=true
    Module.Write SpriteSize=true
    Position=&SimulationPosition()
    Sprite Rotation=&UniformRangedFloat(Minimum=-180, Maximum=180)
    Sprite Size=&Vector2DFromFloat(Value=1)
   )]
  particleUpdateScript [
   &ParticleState()
   &Color(Color=&ColorFromCurve(ColorCurve=DataInterface() {propertyValues {
       AlphaCurve {keys [<at 0 1>, {arriveTangent=-1, interpMode=RCIM_Cubic, leaveTangent=-1, time=0.2464, value=0.9}, <at 1 0>]}
       BlueCurve {keys [{arriveTangent=-1.0571, interpMode=RCIM_Cubic, time=0.4549, value=0.4}, <at 0.8426 1>]}
       GreenCurve {keys [{arriveTangent=-5.7083, interpMode=RCIM_Cubic, time=0.4549, value=0.73}, <at 0.8426 1>]}
       RedCurve {keys [{arriveTangent=-9.5138, interpMode=RCIM_Cubic, time=0.4549, value=1}, <at 0.8426 1>]}
      }}))
   &ScaleColor(Color Value To Scale=$Particles.Color, ScaleA=true, ScaleRGB=true)
   &SkeletalMeshLocation(
    Apply Sampled Orientation to Mesh Particles=true
    Bone Sampling Mode=::NewEnumerator4
    Inherit Velocity=false
    MaskSpawnGroup=false
    Sampled Position Offset=<0 0 -10>
    Skeletal Mesh=$User.User Skeletal Mesh
   )
   &CameraOffset(Camera Offset Amount=10)
   &DynamicMaterialParameters(
    Index 0 Param 1=&FloatFromCurve(FloatCurve=DataInterface() {propertyValues {Curve {keys [<at 0 0.8>,<at 1 1.5>]}}})
    Module.Param0WriteEnabled=true
    Module.Param1WriteEnabled=false
    Module.Param2WriteEnabled=false
    Module.Param3WriteEnabled=false
   )
   &ScaleMeshSize(Scale Factor=&VectorFromFloat(Value=&FloatFromCurve(FloatCurve=DataInterface() {propertyValues {Curve {keys [{interpMode=RCIM_Cubic, leaveTangent=0.2946, tangentMode=RCTM_Break, time=0, value=0}, <at 1 1>]}}})))
  ]
  rendererClasses [&NiagaraMeshRendererProperties]
  renderers [{rendererClass=&NiagaraMeshRendererProperties, rendererIndex=0}]
 }}
summary {
 systemName=NS_Glider_Historian_Whirl
 emitters [{emitterName=Ring, rendererClasses [&NiagaraMeshRendererProperties]}]
 userVariables [
  User.Secondary Color {type=LinearColor, description="", defaultValue {struct=LinearColor, value=<rgba 1 1 1 1>}}
  User.User Skeletal Mesh {
   type=NiagaraDataInterfaceSkeletalMesh
   defaultValue {
    struct=NiagaraExt_VariableValue_DataInterface
    value {dataInterface=&self:NiagaraDataInterfaceSkeletalMesh_0}
   }
  }
 ]
}
